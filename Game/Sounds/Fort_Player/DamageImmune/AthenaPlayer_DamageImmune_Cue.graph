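# Sound cue for the Athena player's damage-immune sound, played at reduced volume in the player attacks and abilities sound class.
# A Random node picks one of three equally weighted wave variants and is excluded from branch culling.
# A root Modulator holds volume fixed, leaving pitch modulation at its defaults.
# Dedicated attenuation and a concurrency override limit range and simultaneous instances, with virtualization disabled.
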SoundCue /Game/Sounds/Fort_Player/DamageImmune/AthenaPlayer_DamageImmune_Cue
root=Modulator_0
nodes [
 {id=Modulator_0, children [Random_0], properties {volumeMax=1, volumeMin=1}}
 {id=Random_0, children [WavePlayer_1,WavePlayer_2,WavePlayer_3], properties {bShouldExcludeFromBranchCulling=true, weights [1,1,1]}}
 {id=WavePlayer_1, properties {soundWaveAssetPtr=&/AthenaPlayer_DamageImmune_01}}
 {id=WavePlayer_2, properties {soundWaveAssetPtr=&/AthenaPlayer_DamageImmune_02}}
 {id=WavePlayer_3, properties {soundWaveAssetPtr=&/AthenaPlayer_DamageImmune_03}}
]
properties {
 attenuationSettings=&/AthenaPlayer_DamageImmune_Att
 bOverrideConcurrency=true
 maxDistance=3650
 soundClassObject=/Game/Sounds/FortSoundClassesAndModes/Fort_06_Player_Attacks_Abilities
 virtualizationMode=Disabled
 volumeMultiplier=0.3
 concurrencyOverrides {maxCount=2}
}
